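; Creative Speaker device that plays a selectable sound when damaged, hit by a music sequencer or activated by other devices.
; Property changes async-load the cue and set volume and attenuation; minigame phase and enable/disable events gate triggering, along with retrigger timing.
; The server only stamps LastPlayed, whose OnRep starts playback, offset during replays.
; A replicated registered-player list and knobs decide who hears it and whether it plays at the device or attached to players.
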(event OnAnyPropertyChanged(ToyOptionsComponent)
  ;[0 1 0.3936 0.36] Initialization
  (CallFunction|UpdateAudioSample) ; Update on server as well as we need to know duration\nfor determining if we can retrigger.
  (Utilities|FlowControl|Sequence
    (:then_0
      (Networking|SwitchHasAuthority
        (:Authority
          (Visibility|SetVisiblityDuringPhase (Variables|Default|GetVisibleInGameComponent) (VisibleinGameComponent|GetVisibilityDuringPhaseEnum)))))
    (:then_1
      (bind _currentminigame (Variables|/Script/FortniteGame.FortMinigameLogicComponent|GetCurrentMinigame (Variables|Default|GetFortMinigameProgress))
        (:then
          (CallFunction|UpdateIsEnabled (Variables|/Script/FortniteGame.FortMinigame|GetCurrentState _currentminigame)))
        (:else)))))

(event OnPropertiesLoaded(ToyOptionsComponent)
  ;[0 1 0.3936 0.36] Initialization
  (CallFunction|UpdateAudioSample) ; Update on server as well as we need to know duration\nfor determining if we can retrigger.
  (Utilities|FlowControl|Sequence
    (:then_0
      (Networking|SwitchHasAuthority
        (:Authority
          (Visibility|SetVisiblityDuringPhase (Variables|Default|GetVisibleInGameComponent) (VisibleinGameComponent|GetVisibilityDuringPhaseEnum)))))
    (:then_1
      (bind _currentminigame (Variables|/Script/FortniteGame.FortMinigameLogicComponent|GetCurrentMinigame (Variables|Default|GetFortMinigameProgress))
        (:then
          (CallFunction|UpdateIsEnabled (Variables|/Script/FortniteGame.FortMinigame|GetCurrentState _currentminigame)))
        (:else)))))

(event Custom|Update%20Audio%20Sample
  ;[3 1.222 0 0.815] Update Audio Values
  (Audio|Components|Audio|Stop (Variables|Default|GetAudio))
  (if (!= (Variables|DeviceKnobs|GetAudioIndex) (Variables|Default|GetLoaded%20Audio%20Index))
    (if (!= (Variables|DeviceKnobs|GetAudioIndex) 0)
      (bind _object (Utilities|AsyncLoadAsset (Utilities|Array|Get(acopy) (Variables|Default|GetAudioSamples) (Variables|DeviceKnobs|GetAudioIndex)))
        (:Completed
          (bind _assound_cue (Utilities|Casting|CastToSoundCue _object))
          (Audio|Components|Audio|SetSound (Variables|Default|GetAudio) _assound_cue)))
      (else
        (Audio|Components|Audio|SetSound (Variables|Default|GetAudio) 0) ; Index is 0, no audio.
      ))
    (Variables|Default|SetLoaded%20Audio%20Index (Variables|DeviceKnobs|GetAudioIndex))
    (Audio|Components|Audio|SetVolumeMultiplier (Variables|Default|GetAudio) (Utilities|Array|Get(acopy) (Variables|Default|GetAlarmVolumeArray) (Variables|DeviceKnobs|GetVolumeIndex)) :!class /Script/Engine.AudioComponent)
    (Variables|/Script/Engine.AudioComponent|SetAttenuationSettings (Utilities|Array|Get(acopy) (Variables|Default|GetAttenuation%20Settings) (Variables|DeviceKnobs|GetAttenuationIndex)) (Variables|Default|GetAudio))
    (else
      _)))

(event MinigameStateChangedDelegate(FortMinigameProgress)
  (params (Minigame :type Fort%20Minigame%20Object%20Reference) (NewMinigameState :type EFortMinigameState%20Enum :default "PreGame"))
  ;[0 1 0.3936 0.36] Initialization
  (CallFunction|UpdateIsEnabled NewMinigameState))

(event OnAddedtoMinigame(FortMinigameProgress)
  (params (Minigame :type Fort%20Minigame%20Object%20Reference))
  ;[0 1 0.3936 0.36] Initialization
  (Utilities|IsValid Minigame
    (:Is%20Valid
      (CallFunction|UpdateIsEnabled (Variables|/Script/FortniteGame.FortMinigame|GetCurrentState Minigame)))))

(event EventOnDamageServer
  (params (Damage :type Float%20%28single-precision%29) (DamageTags :type Gameplay%20Tag%20Container%20Structure%20%28by%20ref%29) (Momentum :type Vector :default "0, 0, 0") (HitInfo :type Hit%20Result%20Structure%20%28by%20ref%29) (InstigatedBy :type Controller%20Object%20Reference) (DamageCauser :type Actor%20Object%20Reference) (EffectContext :type Gameplay%20Effect%20Context%20Handle%20Structure))
  ;[1 1 1] On Hit, trigger audio if we can and want to
  (Utilities|IsValid InstigatedBy
    (:Is%20Valid
      (Health|SetHealth 1e+02) ; Essentially makes this invincible\nwhile leaving it targetable.
      (if (and (CallFunction|AudioCanTrigger) (!= (Variables|Default|GetDamageBehaviour) 0))
        (CallFunction|ServerStartAudio)))))

(event EventOnHitByMusicSequencer
  (params (InstigatedBy :type Controller%20Object%20Reference))
  ;[0 0.3855 1] Play Audio when Triggered
  (Networking|SwitchHasAuthority
    (:Authority
      ;[0.7957 0.005113 1] Gameplay Functions
      (if (CallFunction|AudioCanTrigger)
        (CallFunction|ServerStartAudio)))))

(event Custom|Event%20On%20Receive%20Activate
  (params (Instigator :type Controller%20Object%20Reference))
  ;[0.7957 0.005113 1] Gameplay Functions
  (if (CallFunction|AudioCanTrigger)
    (CallFunction|ServerStartAudio)))

(event Custom|Event%20On%20Receive%20Enable
  (params (Instigator :type Controller%20Object%20Reference))
  ;[0.7957 0.005113 1] Gameplay Functions
  (Variables|Default|SetEnabled true))

(event Custom|Event%20On%20Receive%20Disable
  (params (Instigator :type Controller%20Object%20Reference))
  ;[0.7957 0.005113 1] Gameplay Functions
  (Variables|Default|SetEnabled false))

(event OnVisibilityChanged(VisibleInGameComponent)
  (params (Visible :type Boolean))
  ;[0 0.668 1] Updating Collision when Visibility Changes. Specific rather than letting the component handle it as we want to keep overlap collision so the device can be activated by overlap triggers (Random Number Generator / Sequencer)
  (Collision|SetCollisionProfileName (Variables|Default|GetMesh) (select Visible (Variables|Collision|GetVisibleCollisionProfile) (Variables|Collision|GetInvisibleCollisionProfile)) false :!class /Script/Engine.PrimitiveComponent))

(event Custom|Event%20On%20Receive%20Register%20Player
  (params (Instigator :type Controller%20Object%20Reference))
  (bind _playerstateswithaudioattached (Variables|Default|GetPlayerStatesWithAudioAttached))
  ;[1 1 1] Register/Unregister Instigators
  (Utilities|Array|AddUnique _playerstateswithaudioattached (Variables|/Script/Engine.Controller|GetPlayerState Instigator))
  (Variables|Default|SetRepPlayerStatesWithAudioAttached _playerstateswithaudioattached))

(event Custom|Event%20On%20Receive%20Unregister%20Player
  (params (Instigator :type Controller%20Object%20Reference))
  (bind _playerstateswithaudioattached (Variables|Default|GetPlayerStatesWithAudioAttached))
  ;[1 1 1] Register/Unregister Instigators
  (Utilities|Array|RemoveItem _playerstateswithaudioattached (Variables|/Script/Engine.Controller|GetPlayerState Instigator))
  (Variables|Default|SetRepPlayerStatesWithAudioAttached _playerstateswithaudioattached))

(event Custom|Event%20On%20Receive%20Unregister%20All%20Players
  (params (Instigator :type Controller%20Object%20Reference))
  (bind _playerstateswithaudioattached (Variables|Default|GetPlayerStatesWithAudioAttached))
  ;[1 1 1] Register/Unregister Instigators
  (Utilities|Array|Clear _playerstateswithaudioattached)
  (Variables|Default|SetRepPlayerStatesWithAudioAttached _playerstateswithaudioattached))

(event MinigameStartedDelegate(FortMinigameProgress)
  ;[1 1 1] Bind Delegates to Minigame started and ended
  (Audio|Components|Audio|Stop (Variables|Default|GetAudio))
  (Minigame|BindEventtoPlayerRemovedDelegate (Minigame|GetMinigame self :!class /Script/FortniteGame.MinigameBlueprintLibrary) (delegate Custom|On%20Player%20Removed)))

(event Custom|On%20Player%20Removed
  (params (UniqueNetId :type Unique%20Net%20Id%20Structure) (bIsLocalPlayer :type Boolean))
  ;[1 1 1] Bind Delegates to Minigame started and ended
  (bind _returnvalue (Utilities|Array|RemoveItem (Variables|Default|GetPlayerStatesWithAudioAttached) (Class|CreativeCommonDeviceFunctionLibrary|GetPlayerStateFromNetID UniqueNetId)))
  (if _returnvalue
    (Variables|Default|SetRepPlayerStatesWithAudioAttached (Variables|Default|GetPlayerStatesWithAudioAttached))))

(event MinigameEndedDelegate(FortMinigameProgress)
  ;[1 1 1] Bind Delegates to Minigame started and ended
  (Audio|Components|Audio|Stop (Variables|Default|GetAudio))
  (Minigame|UnbindEventfromPlayerRemovedDelegate (Minigame|GetMinigame self :!class /Script/FortniteGame.MinigameBlueprintLibrary) (EventDispatchers|CreateEvent :!event Custom|On%20Player%20Removed))
  (Utilities|Array|Clear (Variables|Default|GetPlayerStatesWithAudioAttached))
  (Utilities|Array|Clear (Variables|Default|GetRepPlayerStatesWithAudioAttached)))

(fn ActivateFromOtherDevice
  (params (Instigator :type Controller%20Object%20Reference) (Requester :type ERequesterType%20Enum :default "NewEnumerator0"))
  (Networking|SwitchHasAuthority
    (:Authority
      (if (CallFunction|AudioCanTrigger)
        (CallFunction|ServerStartAudio)
        (return true)))))

(fn AudioCanTrigger
  (bind _audio (Variables|Default|GetAudio))
  (bind _sound (Variables|/Script/Engine.AudioComponent|GetSound _audio))
  (Utilities|IsValid _sound
    (:Is%20Valid
      (Utilities|Casting|CastToSoundCue _sound)
      (if (Utilities|Equal(Object) _sound (Variables|Default|GetLastSoundPlayed))
        (return (and (> (Utilities|Time|GetGameTimeinSeconds) (+ (select (== (Variables|DeviceKnobs|GetRetriggerBehaviour) 1) (Variables|Default|GetMin%20Retrigger%20Time) (Variables|/Script/Engine.SoundBase|GetDuration (Variables|/Script/Engine.AudioComponent|GetSound (Variables|Default|GetAudio)))) (Variables|Default|GetLastPlayed))) (Variables|Default|GetEnabled)))
        (else
          (return (Variables|Default|GetEnabled)))))))

(fn CanInteract
  (params (InteractingPawn :type Fort%20Pawn%20Object%20Reference) (InteractionBeingAttempted :type EInteractionBeingAttempted%20Enum :default "FirstInteraction") (InteractionType :type TInteractionType%20Enum :default "IT_NoInteraction"))
  (return false))

(fn GetDeviceColor
  (return 0))

(fn GetInstigatingController
  (return 0))

(fn GetMusicSequencerHitSound
  (params (InstigatedBy :type Controller%20Object%20Reference))
  (return 0 0 0 0))

(fn GetMusicSequencerHitSoundLocation
  (params (InstigatedBy :type Controller%20Object%20Reference))
  (return "0, 0, 0"))

(fn GetSpawnedPickups
  (return 0))

(fn GetVisibilityComponents
  (return (Utilities|Array|MakeArray (Variables|Default|GetMesh))))

(fn GetVisibilityDuringPhaseEnum
  (return (select (and (Variables|Default|GetbIsVisible) (Class|CreativeCommonDeviceFunctionLibrary|IsonUnpublishedIsland self)) "NewEnumerator3" (select (Variables|Default|GetbIsVisible) "NewEnumerator2" (select (Class|CreativeCommonDeviceFunctionLibrary|IsonUnpublishedIsland self) "NewEnumerator1" "NewEnumerator0")))))

(fn IsEnabled
  (return false))

(fn IsHoldingItemForPawn
  (params (Item :type Item%20Variant%20Handle%20Structure) (Pawn :type Pawn%20Object%20Reference))
  (return false))

(fn IsTargetableByLinearMover
  (return true))

(fn OnRep_LastPlayed
  (bind _lastplayed (Variables|Default|GetLastPlayed))
  (bind _returnvalue (Utilities|Time|GetGameTimeinSeconds))
  (bind _returnvalue_2 (- _returnvalue _lastplayed))
  (bind _audio (Variables|Default|GetAudio))
  (bind _sound (Variables|/Script/Engine.AudioComponent|GetSound _audio))
  (Utilities|IsValid _sound
    (:Is%20Valid
      (bind _assound_cue (Utilities|Casting|CastToSoundCue _sound))
      (if (< _returnvalue_2 (Variables|/Script/Engine.SoundBase|GetDuration _assound_cue))
        (CallFunction|PlayAudio (select (Replay|GetIsPlayingReplay) _returnvalue_2 0.0))))))

(fn PlayAudio
  (params (StartTime :type Float%20%28double-precision%29))
  (bind _returnvalue (Fortnite|GetFirstLocalFortPlayerController true))
  (bind _playerstate (Variables|/Script/Engine.Controller|GetPlayerState _returnvalue))
  (bind _repplayerstateswithaudioattached (Variables|Default|GetRepPlayerStatesWithAudioAttached))
  (bind _returnvalue_2 (Utilities|Array|ContainsItem _repplayerstateswithaudioattached _playerstate))
  (bind _currentplayinginstigatorsound (Variables|Default|GetCurrentPlayingInstigatorSound))
  (bind _audio (Variables|Default|GetAudio))
  ;[1 1 1] Check if we have instigators stored, if not check and see if everyone should hear this sound, if so check to see if we play at device
  (Variables|/Script/Engine.AudioComponent|GetSound _audio
    (:then
      (if (Class|CreativeCommonDeviceFunctionLibrary|IsinPlayMode self)
        (if (Utilities|IsArrayNonEmpty (Variables|Default|GetRepPlayerStatesWithAudioAttached))
          ;[1 1 1] Run through instigators to see: Who should be able to hear the Audio, and then where it should be played (on the device or instigator)
          (for _array_element (Variables|Default|GetRepPlayerStatesWithAudioAttached)
            (Utilities|IsValid _array_element
              (:Is%20Valid
                (LocalVariables|SetCurrentLocalPlayerState _array_element)
                (if (== (Variables|DeviceKnobs|GetCanBeheardByIndex) 0)
                  _
                  (elif (and (== (Variables|DeviceKnobs|GetCanBeheardByIndex) 1) _returnvalue_2)
                    _
                    (elif (and (== (Variables|DeviceKnobs|GetCanBeheardByIndex) 2) (not _returnvalue_2))
                      (if (== (Variables|DeviceKnobs|GetPlayatLocationIndex) 1)
                        (Utilities|IsValid _currentplayinginstigatorsound
                          (:Is%20Valid
                            (Audio|Components|Audio|Stop _currentplayinginstigatorsound))
                          (:Is%20Not%20Valid
                            _))
                        (bind _returnvalue_3 (Audio|SpawnSoundAttached :Sound (Variables|/Script/Engine.AudioComponent|GetSound (Variables|Default|GetAudio)) :AttachToComponent (Variables|/Script/Engine.Actor|GetRootComponent (Variables|/Script/Engine.PlayerState|GetPawnPrivate (LocalVariables|GetCurrentLocalPlayerState))) :AttachPointName "None" :Location (Transformation|GetActorLocation (Variables|/Script/Engine.PlayerState|GetPawnPrivate (LocalVariables|GetCurrentLocalPlayerState))) :LocationType "SnapToTarget" :bStopWhenAttachedToDestroyed true :VolumeMultiplier (Variables|/Script/Engine.AudioComponent|GetVolumeMultiplier (Variables|Default|GetAudio)) :PitchMultiplier (Variables|/Script/Engine.AudioComponent|GetPitchMultiplier (Variables|Default|GetAudio)) :StartTime (Variables|GetStartTime) :AttenuationSettings (Variables|/Script/Engine.AudioComponent|GetAttenuationSettings (Variables|Default|GetAudio)) :bAutoDestroy true))
                        (Variables|Default|SetCurrentPlayingInstigatorSound _returnvalue_3)
                        (Variables|Default|SetLastSoundPlayed (Variables|/Script/Engine.AudioComponent|GetSound _returnvalue_3))
                        (else
                          (Audio|Components|Audio|Play (Variables|Default|GetAudio) (Variables|GetStartTime))))))))))
          (elif (or (== 0 (Variables|DeviceKnobs|GetCanBeheardByIndex)) (== 2 (Variables|DeviceKnobs|GetCanBeheardByIndex)))
            (if (== 0 (Variables|DeviceKnobs|GetPlayatLocationIndex))
              _)))
        (else
          _))
      (Audio|Components|Audio|Play _audio StartTime)
      (Variables|Default|SetLastSoundPlayed (Variables|/Script/Engine.AudioComponent|GetSound _audio)))
    (:else)))

(fn ServerStartAudio
  (Variables|Default|SetLastPlayed (Utilities|Time|GetGameTimeinSeconds)))

(fn SetTriggerCollision
  (params (On :type Boolean))
  (return false))

(fn ShouldInvisiblePropsBeAllowed
  (return false))

(fn UpdateVisibilitySettings)

(fn UpdateIsEnabled
  (params (CurrentMinigameState :type EFortMinigameState%20Enum :default "PreGame"))
  (Variables|Default|SetEnabled (Class|CreativeCommonDeviceFunctionLibrary|IsEnabledPhaseMatch CurrentMinigameState (Variables|DeviceKnobs|GetEnabledIndex) self)))

(fn ConstructionScript)
